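# Sound cue for the activation of the Kor boss's Void Run ability.
# One of three activation waves is mixed equally with one of six MaskMaker boss synth whooshes pitched up by a fixed amount.
# IsLocal and IsEnemy branches route this shared mix at full volume for local, slightly lower for enemy or unset, lowest for non-enemy.
# All three paths use the same Void Run on/off attenuation.
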SoundCue /CloakGameplay/Audio/SFX/Alternate/SC_KorBoss_VoidRun_Activate
root=Branch_0
nodes [
 {id=Attenuation_0, children [Mixer_0], properties {attenuationSettings=&/Att_KorBoss_VoidRun_OnOff}}
 {id=Attenuation_1, children [Mixer_1], properties {attenuationSettings=&Att_KorBoss_VoidRun_OnOff}}
 {id=Attenuation_2, children [Mixer_2], properties {attenuationSettings=&Att_KorBoss_VoidRun_OnOff}}
 {id=Branch_0, children [Attenuation_0, Branch_1, Branch_1], properties {boolParameterName=IsLocal}}
 {id=Branch_1, children [Attenuation_1, Attenuation_2, Attenuation_1], properties {boolParameterName=IsEnemy}}
 {id=Mixer_0, children [Mixer_3], properties {inputVolume [1]}}
 {id=Mixer_1, children [Mixer_3], properties {inputVolume [0.9]}}
 {id=Mixer_2, children [Mixer_3], properties {inputVolume [0.8]}}
 {id=Mixer_3, children [Random_0,Modulator_0], properties {inputVolume [0.5,0.5]}}
 {id=Modulator_0, children [Random_2], properties {pitchMax=1.5, pitchMin=1.5}}
 {id=Random_0, children [WavePlayer_3,WavePlayer_4,WavePlayer_5], properties {weights [1,1,1]}}
 {
  id=Random_2
  children [WavePlayer_6, WavePlayer_7, WavePlayer_8, WavePlayer_9, WavePlayer_10, WavePlayer_11]
  properties {weights [1,1,1,1,1,1]}
 }
 {id=WavePlayer_10, properties {soundWaveAssetPtr=&/MaskMakerBoss_SynthWhooshA_02}}
 {id=WavePlayer_11, properties {soundWaveAssetPtr=&/MaskMakerBoss_SynthWhooshA_01}}
 {id=WavePlayer_3, properties {soundWaveAssetPtr=&/KorBoss_VoidRun_Activate_01}}
 {id=WavePlayer_4, properties {soundWaveAssetPtr=&/KorBoss_VoidRun_Activate_02}}
 {id=WavePlayer_5, properties {soundWaveAssetPtr=&/KorBoss_VoidRun_Activate_03}}
 {id=WavePlayer_6, properties {soundWaveAssetPtr=&/MaskMakerBoss_SynthWhooshA_06}}
 {id=WavePlayer_7, properties {soundWaveAssetPtr=&/MaskMakerBoss_SynthWhooshA_05}}
 {id=WavePlayer_8, properties {soundWaveAssetPtr=&/MaskMakerBoss_SynthWhooshA_04}}
 {id=WavePlayer_9, properties {soundWaveAssetPtr=&/MaskMakerBoss_SynthWhooshA_03}}
]
properties {
 attenuationSettings=&Att_KorBoss_VoidRun_OnOff
 maxDistance=6300
 soundClassObject=&Fort_Gadgets_Items
 volumeMultiplier=0.4
}
