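# Sound cue for the deactivation sound of KorBoss Void Run.
# IsLocal and IsEnemy branches pick one of three volume levels (local loudest, enemy or unset next, non-enemy quietest), all using the same attenuation asset.
# Each level feeds one shared mix of a randomly chosen deactivate wave, from two variants, over a quieter layer of six synth whooshes played together with a fixed raised pitch.
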SoundCue /CloakGameplay/Audio/SFX/Alternate/SC_KorBoss_VoidRun_Deactivate
root=Branch_0
nodes [
 {id=Attenuation_0, children [Mixer_0], properties {attenuationSettings=&/Att_KorBoss_VoidRun_OnOff}}
 {id=Attenuation_1, children [Mixer_1], properties {attenuationSettings=&Att_KorBoss_VoidRun_OnOff}}
 {id=Attenuation_2, children [Mixer_2], properties {attenuationSettings=&Att_KorBoss_VoidRun_OnOff}}
 {id=Branch_0, children [Attenuation_0, Branch_1, Branch_1], properties {boolParameterName=IsLocal}}
 {id=Branch_1, children [Attenuation_1, Attenuation_2, Attenuation_1], properties {boolParameterName=IsEnemy}}
 {id=Mixer_0, children [Mixer_3], properties {inputVolume [1]}}
 {id=Mixer_1, children [Mixer_3], properties {inputVolume [0.9]}}
 {id=Mixer_2, children [Mixer_3], properties {inputVolume [0.8]}}
 {id=Mixer_3, children [Random_0,Modulator_0], properties {inputVolume [0.5,0.2]}}
 {
  id=Mixer_4
  children [WavePlayer_10, WavePlayer_11, WavePlayer_12, WavePlayer_13, WavePlayer_14, WavePlayer_15]
  properties {inputVolume [1,1,1,1,1,1]}
 }
 {id=Modulator_0, children [Mixer_4], properties {pitchMax=1.2, pitchMin=1.2}}
 {id=Random_0, children [WavePlayer_1,WavePlayer_2], properties {weights [1,1]}}
 {id=WavePlayer_1, properties {soundWaveAssetPtr=&/KorBoss_VoidRun_Deactivate_01}}
 {id=WavePlayer_10, properties {soundWaveAssetPtr=&/KorBoss_SynthWhooshB_01}}
 {id=WavePlayer_11, properties {soundWaveAssetPtr=&/KorBoss_SynthWhooshB_02}}
 {id=WavePlayer_12, properties {soundWaveAssetPtr=&/KorBoss_SynthWhooshB_03}}
 {id=WavePlayer_13, properties {soundWaveAssetPtr=&/KorBoss_SynthWhooshB_04}}
 {id=WavePlayer_14, properties {soundWaveAssetPtr=&/KorBoss_SynthWhooshB_05}}
 {id=WavePlayer_15, properties {soundWaveAssetPtr=&/KorBoss_SynthWhooshB_06}}
 {id=WavePlayer_2, properties {soundWaveAssetPtr=&/KorBoss_VoidRun_Deactivate_03}}
]
properties {
 attenuationSettings=&Att_KorBoss_VoidRun_OnOff
 maxDistance=6300
 soundClassObject=&Fort_Gadgets_Items
 volumeMultiplier=0.35
}
